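# First-person close sound cue for the alien laser weapon's firing loop.
# A mixer layers a looping fire wave with one of four randomly chosen fire shots, which a modulator raises to a fixed higher pitch; its first input is empty.
# The mix runs through first-person projectile attenuation and a ranged-fire 1P sound class override, with first-person gunfire concurrency on the cue.
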SoundCue /MotherGameplay/Sounds/AlienLaser/AlienLaser_FireLoop_Close1P_Cue
root=SoundClass_1
nodes [
 {
  id=Attenuation_1
  children [Mixer_0]
  properties {attenuationSettings=&Fort_Weapon_Projectile_FirstPerson_Athena_Att}
 }
 {id=Mixer_0, children [null,Modulator_0,WavePlayer_0], properties {inputVolume [0.65,0.55,1]}}
 {id=Modulator_0, children [Random_2], properties {pitchMax=1.1, pitchMin=1.1}}
 {id=Random_2, children [WavePlayer_9, WavePlayer_10, WavePlayer_11, WavePlayer_12], properties {weights [1,1,1,1]}}
 {
  id=SoundClass_1
  children [Attenuation_1]
  properties {soundClassOverride=&Fort_WeaponRanged_Fire_1P_PSM}
 }
 {id=WavePlayer_0, properties {bLooping=true, soundWaveAssetPtr=&/AlienLaser_FireLoop_Close_01a}}
 {id=WavePlayer_10, properties {soundWaveAssetPtr=&/AlienLaser_Fire_Close_02}}
 {id=WavePlayer_11, properties {soundWaveAssetPtr=&/AlienLaser_Fire_Close_03}}
 {id=WavePlayer_12, properties {soundWaveAssetPtr=&/AlienLaser_Fire_Close_04}}
 {id=WavePlayer_9, properties {soundWaveAssetPtr=&/AlienLaser_Fire_Close_01}}
]
properties {bDebug=true, maxDistance=1800.0001, volumeMultiplier=0.8, concurrencySet [&Weapon_Gunfire_FirstPerson_Conc]}
