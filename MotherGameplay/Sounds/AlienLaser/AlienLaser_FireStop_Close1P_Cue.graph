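# Sound cue for the close-range, first-person fire-stop sound of the AlienLaser weapon.
# A random node picks one of three equally weighted wave variations, feeding first-person projectile attenuation and then a ranged-weapon fire sound class override at the root.
# The cue plays with raised pitch and reduced volume, uses first-person gunfire concurrency, and has its debug flag left on.
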SoundCue /MotherGameplay/Sounds/AlienLaser/AlienLaser_FireStop_Close1P_Cue
root=SoundClass_1
nodes [
 {
  id=Attenuation_1
  children [Random_1]
  properties {attenuationSettings=&Fort_Weapon_Projectile_FirstPerson_Athena_Att}
 }
 {id=Random_1, children [WavePlayer_2,WavePlayer_3,WavePlayer_4], properties {weights [1,1,1]}}
 {
  id=SoundClass_1
  children [Attenuation_1]
  properties {soundClassOverride=&Fort_WeaponRanged_Fire_1P_PSM}
 }
 {id=WavePlayer_2, properties {soundWaveAssetPtr=&/AlienLaser_FireStop_Close_01}}
 {id=WavePlayer_3, properties {soundWaveAssetPtr=&/AlienLaser_FireStop_Close_02}}
 {id=WavePlayer_4, properties {soundWaveAssetPtr=&/AlienLaser_FireStop_Close_03}}
]
properties {
 bDebug=true
 maxDistance=1800.0001
 pitchMultiplier=1.2
 volumeMultiplier=0.2
 concurrencySet [&Weapon_Gunfire_FirstPerson_Conc]
}
