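# Sound cue for the alien laser weapon's fire-stop sound, using the third-person ranged-fire sound class on both layers.
# A mixer plays two layers together, each picking one of three equally weighted waves: Close variants through close projectile attenuation, Distant variants through medium mono attenuation.
# The cue raises pitch, lowers volume, joins local and global gunfire concurrency groups, and has its debug flag enabled.
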SoundCue /MotherGameplay/Sounds/AlienLaser/AlienLaser_FireStop_Close_Cue
root=Mixer_17
nodes [
 {
  id=Attenuation_0
  children [Random_10]
  properties {attenuationSettings=&Fort_Weapon_Projectile_Medium_Mono_Athena_Att}
 }
 {id=Attenuation_1, children [Random_9], properties {attenuationSettings=&Fort_Weapon_Projectile_Close_Athena_Att}}
 {id=Mixer_17, children [SoundClass_1,SoundClass_0], properties {inputVolume [1,1]}}
 {id=Random_10, children [WavePlayer_6,WavePlayer_7,WavePlayer_8], properties {weights [1,1,1]}}
 {id=Random_9, children [WavePlayer_2,WavePlayer_4,WavePlayer_5], properties {weights [1,1,1]}}
 {
  id=SoundClass_0
  children [Attenuation_0]
  properties {soundClassOverride=&Fort_WeaponRanged_Fire_3P_PSM}
 }
 {
  id=SoundClass_1
  children [Attenuation_1]
  properties {soundClassOverride=&Fort_WeaponRanged_Fire_3P_PSM}
 }
 {id=WavePlayer_2, properties {soundWaveAssetPtr=&/AlienLaser_FireStop_Close_01}}
 {id=WavePlayer_4, properties {soundWaveAssetPtr=&/AlienLaser_FireStop_Close_02}}
 {id=WavePlayer_5, properties {soundWaveAssetPtr=&/AlienLaser_FireStop_Close_03}}
 {id=WavePlayer_6, properties {soundWaveAssetPtr=&/AlienLaser_FireStop_Distant_01}}
 {id=WavePlayer_7, properties {soundWaveAssetPtr=&/AlienLaser_FireStop_Distant_02}}
 {id=WavePlayer_8, properties {soundWaveAssetPtr=&/AlienLaser_FireStop_Distant_03}}
]
properties {
 bDebug=true
 maxDistance=10579.2871
 pitchMultiplier=1.2
 volumeMultiplier=0.2
 concurrencySet [&Weapon_Gunfire_Local_Conc, &Weapon_Gunfire_Global_Conc]
}
