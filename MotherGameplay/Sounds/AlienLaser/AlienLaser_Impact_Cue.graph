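# Sound cue for alien laser impacts, played in the Fort_Impacts sound class.
# A quality switch sends its first two slots to a full variation set and its last two to a smaller, higher-pitched set.
# Each set mixes random, pitch-modulated laser impacts with half-volume random Atlas spark sounds.
# Overrides add distance attenuation with low-pass filtering and cap concurrency at a few voices, stopping the oldest.
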SoundCue /MotherGameplay/Sounds/AlienLaser/AlienLaser_Impact_Cue
root=QualityLevel_0
nodes [
 {id=Mixer_0, children [Modulator_1,Modulator_0], properties {inputVolume [1,0.5]}}
 {id=Mixer_11, children [Modulator_2,Modulator_3], properties {inputVolume [1,0.5]}}
 {id=Modulator_0, children [Random_1], properties {pitchMax=1.3, pitchMin=1.2}}
 {id=Modulator_1, children [Random_3], properties {pitchMax=1.3, pitchMin=1.2}}
 {id=Modulator_2, children [Random_5], properties {pitchMax=1.1, pitchMin=0.9}}
 {id=Modulator_3, children [Random_14], properties {pitchMax=1.1, pitchMin=0.9}}
 {id=QualityLevel_0, children [Mixer_11,Mixer_11,Mixer_0,Mixer_0]}
 {id=Random_1, children [WavePlayer_2,WavePlayer_3,WavePlayer_4], properties {weights [1,1,1]}}
 {
  id=Random_14
  children [WavePlayer_190, WavePlayer_191, WavePlayer_192, WavePlayer_193, WavePlayer_194, WavePlayer_195]
  properties {weights [1,1,1,1,1,1]}
 }
 {id=Random_3, children [WavePlayer_24, WavePlayer_25, WavePlayer_26, WavePlayer_0, WavePlayer_1], properties {weights [1,1,1,1,1]}}
 {
  id=Random_5
  children [
   WavePlayer_27
   WavePlayer_28
   WavePlayer_29
   WavePlayer_30
   WavePlayer_31
   WavePlayer_32
   WavePlayer_33
   WavePlayer_34
   WavePlayer_35
   WavePlayer_36
   WavePlayer_37
   WavePlayer_38
  ]
  properties {weights [1,1,1,1,1,1,1,1,1,1,1,1]}
 }
 {id=WavePlayer_0, properties {soundWaveAssetPtr=&/AlienLaser_Impact_01}}
 {id=WavePlayer_1, properties {soundWaveAssetPtr=&/AlienLaser_Impact_02}}
 {id=WavePlayer_190, properties {soundWaveAssetPtr=/Game/Sounds/Fort_Atlas_Sounds/atlas_sparkbig_01}}
 {id=WavePlayer_191, properties {soundWaveAssetPtr=/Game/Sounds/Fort_Atlas_Sounds/atlas_sparkbig_04}}
 {id=WavePlayer_192, properties {soundWaveAssetPtr=/Game/Sounds/Fort_Atlas_Sounds/atlas_sparkbig_06}}
 {id=WavePlayer_193, properties {soundWaveAssetPtr=/Game/Sounds/Fort_Atlas_Sounds/atlas_sparkbig_02}}
 {id=WavePlayer_194, properties {soundWaveAssetPtr=/Game/Sounds/Fort_Atlas_Sounds/atlas_sparkbig_03}}
 {id=WavePlayer_195, properties {soundWaveAssetPtr=/Game/Sounds/Fort_Atlas_Sounds/atlas_sparkbig_05}}
 {id=WavePlayer_2, properties {soundWaveAssetPtr=&atlas_sparkbig_01}}
 {id=WavePlayer_24, properties {soundWaveAssetPtr=&/AlienLaser_Impact_10}}
 {id=WavePlayer_25, properties {soundWaveAssetPtr=&/AlienLaser_Impact_11}}
 {id=WavePlayer_26, properties {soundWaveAssetPtr=&/AlienLaser_Impact_12}}
 {id=WavePlayer_27, properties {soundWaveAssetPtr=&AlienLaser_Impact_10}}
 {id=WavePlayer_28, properties {soundWaveAssetPtr=&AlienLaser_Impact_11}}
 {id=WavePlayer_29, properties {soundWaveAssetPtr=&AlienLaser_Impact_12}}
 {id=WavePlayer_3, properties {soundWaveAssetPtr=&atlas_sparkbig_04}}
 {id=WavePlayer_30, properties {soundWaveAssetPtr=&AlienLaser_Impact_01}}
 {id=WavePlayer_31, properties {soundWaveAssetPtr=&AlienLaser_Impact_02}}
 {id=WavePlayer_32, properties {soundWaveAssetPtr=&/AlienLaser_Impact_03}}
 {id=WavePlayer_33, properties {soundWaveAssetPtr=&/AlienLaser_Impact_04}}
 {id=WavePlayer_34, properties {soundWaveAssetPtr=&/AlienLaser_Impact_05}}
 {id=WavePlayer_35, properties {soundWaveAssetPtr=&/AlienLaser_Impact_06}}
 {id=WavePlayer_36, properties {soundWaveAssetPtr=&/AlienLaser_Impact_07}}
 {id=WavePlayer_37, properties {soundWaveAssetPtr=&/AlienLaser_Impact_08}}
 {id=WavePlayer_38, properties {soundWaveAssetPtr=&/AlienLaser_Impact_09}}
 {id=WavePlayer_4, properties {soundWaveAssetPtr=&atlas_sparkbig_06}}
]
properties {
 bOverrideAttenuation=true
 bOverrideConcurrency=true
 maxDistance=2600
 soundClassObject=&Fort_Impacts
 volumeMultiplier=0.8
 attenuationOverrides {
  bAttenuateWithLPF=true
  falloffDistance=2400
  lPFFrequencyAtMax=2000
  lPFRadiusMax=2400
  lPFRadiusMin=800
  stereoSpread=0
  attenuationShapeExtents {x=200}
 }
 concurrencyOverrides {maxCount=5, resolutionRule=StopOldest, volumeScale=0.75}
}
