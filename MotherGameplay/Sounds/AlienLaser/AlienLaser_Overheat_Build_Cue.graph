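# Sound cue for the alien laser weapon's overheat build-up, using the third-person ranged weapon fire sound class on both of its branches.
# The close-attenuated branch layers a pre-fire sound with a delayed overheat build.
# The quieter-mixed, medium-range mono branch layers a delayed distant overheat build, a pitched-down distant fire build and a random pick of two more.
# Both overheat waves start one second after the other layers.
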SoundCue /MotherGameplay/Sounds/AlienLaser/AlienLaser_Overheat_Build_Cue
root=Mixer_2
nodes [
 {id=Attenuation_49, children [Modulator_1], properties {attenuationSettings=&Fort_Weapon_Projectile_Close_Athena_Att}}
 {
  id=Attenuation_5
  children [Modulator_0]
  properties {attenuationSettings=&Fort_Weapon_Projectile_Medium_Mono_Athena_Att}
 }
 {id=Delay_0, children [WavePlayer_12], properties {delayMax=1, delayMin=1}}
 {id=Delay_1, children [WavePlayer_11], properties {delayMax=1, delayMin=1}}
 {id=Mixer_0, children [Delay_1,Modulator_3,Random_1], properties {inputVolume [1,0.8,0.6]}}
 {id=Mixer_1, children [Delay_0,WavePlayer_3,null], properties {inputVolume [0.7,0.6,0.6]}}
 {id=Mixer_2, children [SoundClass_32, SoundClass_5], properties {inputVolume [1,0.4]}}
 {id=Modulator_0, children [Mixer_0], properties {pitchMax=1, pitchMin=1, volumeMax=1.7, volumeMin=1.6}}
 {id=Modulator_1, children [Mixer_1], properties {pitchMax=1, pitchMin=1, volumeMax=1, volumeMin=1}}
 {id=Modulator_3, children [WavePlayer_8], properties {pitchMax=0.9, pitchMin=0.9}}
 {id=Random_1, children [WavePlayer_6,WavePlayer_7], properties {weights [1,1]}}
 {
  id=SoundClass_32
  children [Attenuation_49]
  properties {soundClassOverride=&Fort_WeaponRanged_Fire_3P_PSM}
 }
 {
  id=SoundClass_5
  children [Attenuation_5]
  properties {soundClassOverride=&Fort_WeaponRanged_Fire_3P_PSM}
 }
 {id=WavePlayer_11, properties {soundWaveAssetPtr=&/AlienLaser_Overheat_Build_Distant_01}}
 {id=WavePlayer_12, properties {soundWaveAssetPtr=&/AlienLaser_Overheat_Build_01}}
 {id=WavePlayer_3, properties {soundWaveAssetPtr=&/AlienLaser_PreFire}}
 {id=WavePlayer_6, properties {soundWaveAssetPtr=&/AlienLaser_Fire_Distant_Build_01}}
 {id=WavePlayer_7, properties {soundWaveAssetPtr=&/AlienLaser_Fire_Distant_Build_02}}
 {id=WavePlayer_8, properties {soundWaveAssetPtr=&/AlienLaser_Fire_Distant_Build_03}}
]
properties {maxDistance=10579.2871, volumeMultiplier=0.7}
