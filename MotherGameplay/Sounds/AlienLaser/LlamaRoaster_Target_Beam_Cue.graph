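# Sound cue for the looping alien laser beam audio of the LlamaRoaster target beam.
# A Random node picks between two equally weighted looping fire waves, feeding a SoundClass node that assigns the third-person ranged weapon fire class.
# Inline overrides set distance attenuation with stereo spread and limit concurrent instances per owner, with a low volume multiplier and the debug flag enabled.
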SoundCue /MotherGameplay/Sounds/AlienLaser/LlamaRoaster_Target_Beam_Cue
root=SoundClass_5
nodes [
 {id=Random_1, children [WavePlayer_0,WavePlayer_1], properties {weights [1,1]}}
 {id=SoundClass_5, children [Random_1], properties {soundClassOverride=&Fort_WeaponRanged_Fire_3P_PSM}}
 {id=WavePlayer_0, properties {bLooping=true, soundWaveAssetPtr=&/AlienLaser_FireLoop_Beam_01}}
 {id=WavePlayer_1, properties {bLooping=true, soundWaveAssetPtr=&/AlienLaser_FireLoop_Beam_02}}
]
properties {
 bDebug=true
 bOverrideAttenuation=true
 bOverrideConcurrency=true
 maxDistance=1900
 volumeMultiplier=0.1
 attenuationOverrides {falloffDistance=1500, stereoSpread=512}
 concurrencyOverrides {bLimitToOwner=true, maxCount=2}
}
